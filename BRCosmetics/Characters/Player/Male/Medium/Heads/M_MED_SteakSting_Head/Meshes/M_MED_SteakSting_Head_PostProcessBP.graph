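# Post-process AnimBlueprint for the male medium SteakSting head cosmetic, adding an aiming eyelid pose and secondary tail motion.
# While bIsTargeting is true, an additive bone-space rotation on the left upper mid eyelid eases in, and eases out afterward.
# Four chained Anim Dynamics nodes then simulate dyn_tail_1 to dyn_tail_4 relative to the pelvis, using angular springs, tight limits and custom gravity.
# The EventGraph holds only unconnected stub nodes.
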AnimBlueprint /BRCosmetics/Characters/Player/Male/Medium/Heads/M_MED_SteakSting_Head/Meshes/M_MED_SteakSting_Head_PostProcessBP
target_skeleton=&/M_MED_SteakSting_Head_Skeleton
graphs [
 AnimGraph {nodes [
   n1 = "Transform (Modify) Bone - Bone: L_eye_lid_upper_mid"!(
    ComponentPose=Local To Component(LocalPose=Input Pose())
    bAlphaBoolEnabled:bool=Get bIsTargeting()
    Translation="0,0,0"
    Rotation="9.904,-10.132,-1.320"
    Scale="1,1,1"
   ) -> (Pose) {
    alphaInputType=Bool
    rotationMode=BMM_Additive
    rotationSpace=BCS_BoneSpace
    translationMode=BMM_Additive
    translationSpace=BCS_BoneSpace
    alphaBoolBlend {blendInTime=0.3, blendOption=ExpInOut, blendOutTime=0.3}
    boneToModify {boneName=L_eye_lid_upper_mid}
    rotation=<rot 9.904 -10.132 -1.32>
   }
   n2 = Anim Dynamics - Bone: dyn_tail_1(ComponentPose=@n1, Alpha=1, ExternalForce="0,0,0") {
    angularBiasOverride=0.8
    angularSpringConstant=100000
    bAngularSpring=true
    bOverrideAngularBias=true
    bUseGravityOverride=true
    simulationSpace=BoneRelative
    alphaBoolBlend {blendOption=Linear}
    boundBone {boneName=dyn_tail_1}
    componentAppliedLinearAccClamp=<3000 3000 3000>
    componentLinearAccScale=<0.001 0.001 0.001>
    componentLinearVelScale=<0.001 0.001 0.001>
    gravityOverride {y=-20, z=50}
    physicsBodyDefinitions [{
      boundBone {boneName=dyn_tail_1}
      boxExtents=<20 20 20>
      constraintSetup {angularLimitsMax=<0 10 10>, angularLimitsMin=<0 -10 -10>, angularTarget=<0 0 0>}
      localJointOffset=<5 0 0>
     }]
    relativeSpaceBone {boneName=pelvis}
   }
   n3 = Anim Dynamics - Bone: dyn_tail_2(ComponentPose=@n2, Alpha=1, ExternalForce="0,0,0") {
    angularBiasOverride=0.8
    angularSpringConstant=100000
    bAngularSpring=true
    bOverrideAngularBias=true
    bUseGravityOverride=true
    simulationSpace=BoneRelative
    alphaBoolBlend {blendOption=Linear}
    boundBone {boneName=dyn_tail_2}
    componentAppliedLinearAccClamp=<3000 3000 3000>
    componentLinearAccScale=<0.001 0.001 0.001>
    componentLinearVelScale=<0.001 0.001 0.001>
    gravityOverride {y=-20, z=50}
    physicsBodyDefinitions [{
      boundBone {boneName=dyn_tail_2}
      boxExtents=<20 20 20>
      constraintSetup {angularLimitsMax=<10 10 10>, angularLimitsMin=<-10 -10 -10>, angularTarget=<0 0 0>}
      localJointOffset=<5 0 0>
     }]
    relativeSpaceBone {boneName=pelvis}
   }
   n4 = Anim Dynamics - Bone: dyn_tail_3(ComponentPose=@n3, Alpha=1, ExternalForce="0,0,0") {
    angularBiasOverride=0.8
    angularSpringConstant=100000
    bAngularSpring=true
    bOverrideAngularBias=true
    bUseGravityOverride=true
    simulationSpace=BoneRelative
    alphaBoolBlend {blendOption=Linear}
    boundBone {boneName=dyn_tail_3}
    componentAppliedLinearAccClamp=<3000 3000 3000>
    componentLinearAccScale=<0.001 0.001 0.001>
    componentLinearVelScale=<0.001 0.001 0.001>
    gravityOverride {y=-20, z=50}
    physicsBodyDefinitions [{
      boundBone {boneName=dyn_tail_3}
      boxExtents=<20 20 20>
      constraintSetup {angularLimitsMax=<10 10 10>, angularLimitsMin=<-10 -10 -10>, angularTarget=<0 0 0>}
      localJointOffset=<5 0 0>
     }]
    relativeSpaceBone {boneName=pelvis}
   }
   n5 = Anim Dynamics - Bone: dyn_tail_4(ComponentPose=@n4, Alpha=1, ExternalForce="0,0,0") {
    angularBiasOverride=0.8
    angularSpringConstant=100000
    bAngularSpring=true
    bOverrideAngularBias=true
    bUseGravityOverride=true
    simulationSpace=BoneRelative
    alphaBoolBlend {blendOption=Linear}
    boundBone {boneName=dyn_tail_4}
    componentAppliedLinearAccClamp=<3000 3000 3000>
    componentLinearAccScale=<0.001 0.001 0.001>
    componentLinearVelScale=<0.001 0.001 0.001>
    gravityOverride {y=-20, z=50}
    physicsBodyDefinitions [{
      boundBone {boneName=dyn_tail_4}
      boxExtents=<20 20 20>
      constraintSetup {angularLimitsMax=<10 10 10>, angularLimitsMin=<-10 -10 -10>, angularTarget=<0 0 0>}
      localJointOffset=<5 0 0>
     }]
    relativeSpaceBone {boneName=pelvis}
   }
   Output Pose(Result=Component To Local(ComponentPose=@n5))
  ]}
 EventGraph {nodes [
   TryGetPawnOwner()
   Event BlueprintUpdateAnimation()
  ]}
]
variables []
